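# Translucent unlit material for blue wisp trails on Niagara mesh particles in the TinyTree emote.
# SplineThicken builds ribbon width varied by panning Perlin noise, and nested sine waves add a swirling vertex offset phased by dynamic Param4.
# A trail texture, shifted by particle random and dynamic Position, drives a three-blue emissive gradient and opacity.
# Noise-based SmoothStep erodes both ends via BottomEdge and TopEdge, with particle color and depth fade applied.
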Material /BRCosmetics/Animation/Game/MainPlayer/Emotes/TinyTree/FX/M_Emote_TinyTree_Wisps
graph {
 properties {
  bAutomaticallySetUsageInEditor=false
  bUsedWithNiagaraMeshParticles=true
  blendMode=BLEND_Translucent
  previewMesh=&/SM_Emotes_TinyTree_Soulercoaster_02
  shadingModel=MSM_Unlit
  twoSided=true
  usedShadingModels=Unlit
 }
 nodes [
  n1 = ScalarParameter() {defaultValue=50, parameterName=RibbonTopWidth}
  n2 = TextureSample(UVs=Panner(Coordinate=TextureCoordinate() {vTiling=0}) {speedX=-0.25}) {texture=/Game/Items/Traps/Textures/T_PerlinRGB_01_M}
  n3 = Saturate(None=@n2.R)
  n4 = Multiply(A=@n1, B=@n3)
  n5 = ScalarParameter() {defaultValue=0.2, parameterName=WidthDifferenceStrenght}
  n6 = Multiply(A=ScalarParameter() {defaultValue=55, parameterName=RibbonEndWidth}, B=@n3)
  n7 = TextureCoordinate()
  n8 = Multiply(A=Time()) {constB=0.5}
  n9 = DynamicParameter() {defaultValue=<rgba 1 0 0.1 1>, paramNames [Position,BottomEdge,TopEdge,Param4]}
  n10 = Add(
   A=Multiply(A=ComponentMask(None=@n7) {r=true}, B=ScalarParameter() {defaultValue=0.5, parameterName=WPOTile}) {constB=2}
   B=Add(A=Multiply(A=@n2.R) {constB=0.5}, B=Add(A=@n8, B=@n9.Param4))
  )
  n11 = Multiply(A=Constant3Vector() {constant=<rgba 0 0 1 0>}, B=ScalarParameter() {defaultValue=5, parameterName=RandomOffsetStrenght}) {constB=5}
  n12 = Multiply(A=ComponentMask(None=TextureCoordinate()) {r=true}, B=ScalarParameter() {defaultValue=2, parameterName=SwirlOffsetTileMult}) {constB=2}
  n13 = Multiply(A=Add(A=Time(), B=@n9.Param4), B=ScalarParameter() {defaultValue=1, parameterName=SwirlTimeMult})
  n14 = Multiply(
   A=Sine(None=Add(A=@n12, B=@n13))
   B=Multiply(A=VertexNormalWS(), B=ScalarParameter() {defaultValue=10, parameterName=SwirlOffsetStrenght}) {constB=10}
  )
  Add(
   A=MaterialFunctionCall(
    WidthBase=LinearInterpolate(A=@n1, B=@n4, Alpha=@n5)
    WidthTip=LinearInterpolate(A=@n1, B=@n6, Alpha=@n5)
    Expand U or V UV Channel=StaticBool()
   ) {materialFunction=&SplineThicken}.WorldPositionOffset
   B=Multiply(A=Sine(None=@n10), B=Add(A=@n11, B=@n14))
  )
  n15 = MaterialFunctionCall(Float2=TextureCoordinate() {uTiling=2}) {materialFunction=&BreakOutFloat2Components}
  n16 = Add(A=@n15.R, B=ParticleRandom()) {constB=1.2}
  n17 = Multiply(A=DynamicParameter() {defaultValue=<rgba 1 0 0.1 1>, paramNames [Position,BottomEdge,TopEdge,Param4]}.Position) {constB=-1}
  n18 = TextureSample(UVs=AppendVector(A=Add(A=@n16, B=@n17), B=@n15.G)) {texture=/BRCosmetics/Weapons/FORT_Melee/Pickaxe_Future_Samurai_Summer_Male/FX/T_Pickaxe_Future_Samurai_Summer_Male_Trail_01}
  n19 = MaterialFunctionCall(Float2=TextureCoordinate()) {materialFunction=&BreakOutFloat2Components}
  n20 = Multiply(A=OneMinus(None=@n19.R), B=Saturate(None=Multiply(A=@n19.R) {constB=15}))
  n21 = ComponentMask(None=TextureCoordinate()) {r=true}
  n22 = DynamicParameter() {defaultValue=<rgba 1 0 0.1 1>, paramNames [Position,BottomEdge,TopEdge,Param4]}
  n23 = Power(Base=Subtract(A=OneMinus(None=@n21), B=LinearInterpolate(Alpha=@n22.BottomEdge) {constA=-0.9, constB=0.2})) {constExponent=10}
  n24 = TextureSample(UVs=TextureCoordinate() {vTiling=0.75}) {samplerType=SAMPLERTYPE_Grayscale, texture=/Game/Global/Textures/Default/Noise/LowResBlurredNoiseNormalized}
  n25 = Multiply(
   A=Multiply(A=Multiply(A=@n18.R, B=@n20), B=ScalarParameter() {defaultValue=1.5, parameterName=Param_3})
   B=MaterialFunctionCall(Alpha=Saturate(None=@n23), Min=@n24.R) {materialFunction=&SmoothStep}.Result
  )
  n26 = Subtract(A=ComponentMask(None=TextureCoordinate()) {r=true}, B=LinearInterpolate(Alpha=@n22.TopEdge) {constA=-0.9, constB=0.2})
  n27 = MaterialFunctionCall(Alpha=Saturate(None=Power(Base=@n26) {constExponent=10}), Min=@n24.R) {materialFunction=&SmoothStep}
  n28 = ParticleColor()
  DepthFade(Opacity=Multiply(A=Multiply(A=@n25, B=@n27.Result), B=@n28.A)) {fadeDistanceDefault=10}
  n29 = Constant3Vector() {constant=<rgba 0.0738 0.1204 1 1>}
  n30 = Constant3Vector() {constant=<rgba 0.2137 0.643 1 1>}
  n31 = Constant3Vector() {constant=<rgba 0.5772 0.8753 1 1>}
  Multiply(
   A=MaterialFunctionCall(Emissive=MaterialFunctionCall(A=@n29, B=@n30, C=@n31, Alpha=@n18.R) {materialFunction=&Lerp_3Color}.Result) {materialFunction=&MF_FX_EmissiveAdjustments}.Result
   B=@n28.RGB
  )
  TextureCoordinate() {uTiling=0.2}
 ]
}
