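# Niagara effect for the prop spawn in the TinyTree emote: smoke, snowflake and sparkle bursts, plus mesh wisps and glowing tree ornaments.
# Six colored ornament sprites follow bones on the SpheresMesh skeletal mesh, with a light renderer and twinkling sparkles sampled from the same mesh.
# Nested spawn-group masks give each ornament its own flicker-and-fade alpha curve, scaled by User.OrnamentOpacity and time of day.
# Smoke, snowflakes and wisps fire timed bursts, with a shared beat at 1.8 seconds when the ornaments start.
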NiagaraSystem /BRCosmetics/Animation/Game/MainPlayer/Emotes/TinyTree/FX/NS_Emote_TinyTree_PropSpawn
dependencies {
 usedDataInterfaces [&NiagaraDataInterfaceCurve, &NiagaraDataInterfaceArrayFloat3, &NiagaraDataInterfaceSkeletalMesh]
 usedDynamicInputs []
 usedModules [
  &self:SystemSpawnScript.SystemScriptSource.SystemScriptGraph.NiagaraNodeAssignment_0.SetVariables_#1
  &self:SystemSpawnScript.SystemScriptSource.SystemScriptGraph.NiagaraNodeAssignment_1.SetVariables_#2
  &SystemState
 ]
}
emitters {
 NE_Ornaments_Spheres {
  emitterUpdateScript [
   &EmitterState(Life Cycle Mode=::NewEnumerator1, Loop Behavior=::NewEnumerator1, Loop Delay=1.8, Loop Duration=3, UseLoopDelay=true)
   &SpawnBurst_Instantaneous(Spawn Count=6, Use Spawn Probability=false, UseLoopCountLimit=false)
  ]
  particleSpawnScript [
   &FN_SetSpawnGroup(Number of Spawn Groups=15)
   &InitializeParticle(
    Color=<rgba 1 1 1 0.3>
    Lifetime=5
    Sprite Size Mode=::NewEnumerator3
    Uniform Sprite Size=30
    UsePositionOffset=false
    Write Material Random=true
   )
   &CameraOffset(Camera Offset Amount=10)
  ]
  particleUpdateScript [
   &SkeletalMeshLocation(
    Bone / Socket Index=$Particles.SpawnGroup
    Bone Sampling Mode=::NewEnumerator4
    MaskSpawnGroup=false
    Sampled Position Offset=<0 0 -8>
    Skeletal Mesh=$System.SpheresMesh
   )
   &ParticleState(Kill Particles When Lifetime Has Elapsed=false, Let Infinitely Lived Particles Die When Emitter Deactivates=true)
   /SetVariables(Particles.TImeofDayIntensity:NiagaraFloat=&Multiply_Float(
     A=$User.OrnamentOpacity
     B=&FNTimeOfDayScalar(Night Scale=0.5, Morning Scale=0.7, Evening Scale=0.6)
    )) {isSetParameters=true}
   &ScaleColor(
    Scale Alpha=&MaskFloatBySpawnGroup(
     Masked Float=&FloatFromCurve(
      FloatCurve=DataInterface() {propertyValues {Curve {keys [
          <at 0 0 RCIM_Cubic>
          {arriveTangentWeight=0.0371, interpMode=RCIM_Cubic, leaveTangentWeight=0.7205, tangentMode=RCTM_User, time=0.05, value=1}
          {
           arriveTangent=-0.0287
           arriveTangentWeight=0.1787
           interpMode=RCIM_Cubic
           leaveTangent=-0.0287
           leaveTangentWeight=0.1787
           tangentMode=RCTM_User
           tangentWeightMode=RCTWM_WeightedBoth
           time=2.3
           value=0.975
          }
          <at 2.5 0>
         ]}}}
      CurveIndex=$Particles.Age
      Scale Curve=$Particles.TImeofDayIntensity
     )
     Passthrough Float=&MaskFloatBySpawnGroup(
      Masked Float=&FloatFromCurve(
       FloatCurve=DataInterface() {propertyValues {Curve {keys [
           {interpMode=RCIM_Cubic, leaveTangentWeight=0.0333, tangentMode=RCTM_User, time=0, value=0}
           {arriveTangentWeight=0.0333, interpMode=RCIM_Cubic, leaveTangentWeight=0.6335, tangentMode=RCTM_User, time=0.05, value=1}
           <at 2.249 1>
           <at 2.2832 0.4>
           <at 2.3573 0.4>
           <at 2.3856 1>
           {
            arriveTangent=-0.0244
            arriveTangentWeight=0.1394
            interpMode=RCIM_Cubic
            leaveTangent=-0.0244
            leaveTangentWeight=0.1394
            tangentMode=RCTM_User
            tangentWeightMode=RCTWM_WeightedBoth
            time=2.6
            value=1
           }
           <at 2.7 0>
          ]}}}
       CurveIndex=$Particles.Age
       Scale Curve=$Particles.TImeofDayIntensity
      )
      Passthrough Float=&MaskFloatBySpawnGroup(
       Masked Float=&FloatFromCurve(
        FloatCurve=DataInterface() {propertyValues {Curve {keys [
            <at 0 0 RCIM_Cubic>
            {arriveTangentWeight=0.0371, interpMode=RCIM_Cubic, leaveTangentWeight=0.7205, tangentMode=RCTM_User, time=0.05, value=1}
            {
             arriveTangent=-0.0287
             arriveTangentWeight=0.1787
             interpMode=RCIM_Cubic
             leaveTangent=-0.0287
             leaveTangentWeight=0.1787
             tangentMode=RCTM_User
             tangentWeightMode=RCTWM_WeightedBoth
             time=2.3084
             value=1
            }
            <at 2.5015 0>
           ]}}}
        CurveIndex=$Particles.Age
        Scale Curve=$Particles.TImeofDayIntensity
       )
       Passthrough Float=&MaskFloatBySpawnGroup(
        Masked Float=&FloatFromCurve(
         FloatCurve=DataInterface() {propertyValues {Curve {keys [
             <at 0 0 RCIM_Cubic>
             {arriveTangentWeight=0.0363, interpMode=RCIM_Cubic, leaveTangentWeight=0.8054, tangentMode=RCTM_User, time=0.05, value=1}
             <at 2.2415 1>
             <at 2.3645 0.1473>
             {
              arriveTangent=-0.0938
              arriveTangentWeight=0.2269
              interpMode=RCIM_Cubic
              leaveTangent=-0.0938
              leaveTangentWeight=0.2269
              tangentMode=RCTM_User
              tangentWeightMode=RCTWM_WeightedBoth
              time=2.4416
              value=1
             }
             <at 2.7022 0>
            ]}}}
         CurveIndex=$Particles.Age
         Scale Curve=$Particles.TImeofDayIntensity
        )
        Passthrough Float=&MaskFloatBySpawnGroup(
         Masked Float=&FloatFromCurve(
          FloatCurve=DataInterface() {propertyValues {Curve {keys [
              {interpMode=RCIM_Cubic, leaveTangentWeight=0.0333, tangentMode=RCTM_User, time=0, value=0}
              {arriveTangentWeight=0.0333, interpMode=RCIM_Cubic, leaveTangentWeight=0.6335, tangentMode=RCTM_User, time=0.05, value=1}
              <at 2.254 1>
              <at 2.2973 0.3732>
              <at 2.3629 0.3732>
              <at 2.4285 1>
              {
               arriveTangent=-0.0244
               arriveTangentWeight=0.1394
               interpMode=RCIM_Cubic
               leaveTangent=-0.0244
               leaveTangentWeight=0.1394
               tangentMode=RCTM_User
               tangentWeightMode=RCTWM_WeightedBoth
               time=2.5883
               value=1
              }
              <at 2.7081 0>
             ]}}}
          CurveIndex=$Particles.Age
          Scale Curve=$Particles.TImeofDayIntensity
         )
         Passthrough Float=&FloatFromCurve(
          FloatCurve=DataInterface() {propertyValues {Curve {keys [
              <at 0 0 RCIM_Cubic>
              {arriveTangentWeight=0.0367, interpMode=RCIM_Cubic, leaveTangentWeight=0.7652, tangentMode=RCTM_User, time=0.05, value=1}
              <at 2.24 1>
              <at 2.3197 0.1473>
              {
               arriveTangent=-0.0636
               arriveTangentWeight=0.3335
               interpMode=RCIM_Cubic
               leaveTangent=-0.0636
               leaveTangentWeight=0.3335
               tangentMode=RCTM_User
               tangentWeightMode=RCTWM_WeightedBoth
               time=2.3581
               value=1
              }
              <at 2.623 0>
             ]}}}
          CurveIndex=$Particles.Age
          Scale Curve=$Particles.TImeofDayIntensity
         )
         Spawn Group Mask=4
        )
        Spawn Group Mask=3
       )
       Spawn Group Mask=2
      )
      Spawn Group Mask=1
     )
    )
    Scale RGB=&SelectVectorFromArray(
     Direct Array Index=&SpawnGroup()
     Vector Selection Array=DataInterface() {propertyValues {FloatData [<0.9366 0 1>, <0 1 0.0173>, <0.0313 0.7764 1>, <1 0 0.0823>, <1 0.213 0>, <1 0 0.0823>]}}
    )
    ScaleA=true
    ScaleRGB=true
   ) {module=ScaleColor001}
   &ScaleColor(
    Scale Alpha=&MaskFloatBySpawnGroup(
     Masked Float=&FloatFromCurve(
      FloatCurve=DataInterface() {propertyValues {Curve {keys [
          <at 0 0 RCIM_Cubic>
          {arriveTangentWeight=0.0371, interpMode=RCIM_Cubic, leaveTangentWeight=0.7205, tangentMode=RCTM_User, time=0.05, value=1}
          {
           arriveTangent=-0.0287
           arriveTangentWeight=0.1787
           interpMode=RCIM_Cubic
           leaveTangent=-0.0287
           leaveTangentWeight=0.1787
           tangentMode=RCTM_User
           tangentWeightMode=RCTWM_WeightedBoth
           time=2.3
           value=0.975
          }
          <at 2.5 0>
         ]}}}
      CurveIndex=$Particles.Age
      Scale Curve=$Particles.TImeofDayIntensity
     )
     Passthrough Float=&MaskFloatBySpawnGroup(
      Masked Float=&FloatFromCurve(
       FloatCurve=DataInterface() {propertyValues {Curve {keys [
           {interpMode=RCIM_Cubic, leaveTangentWeight=0.0333, tangentMode=RCTM_User, time=0, value=0}
           {arriveTangentWeight=0.0333, interpMode=RCIM_Cubic, leaveTangentWeight=0.6335, tangentMode=RCTM_User, time=0.05, value=1}
           <at 2.249 1>
           <at 2.2832 0.4>
           <at 2.3573 0.4>
           <at 2.3856 1>
           {
            arriveTangent=-0.0244
            arriveTangentWeight=0.1394
            interpMode=RCIM_Cubic
            leaveTangent=-0.0244
            leaveTangentWeight=0.1394
            tangentMode=RCTM_User
            tangentWeightMode=RCTWM_WeightedBoth
            time=2.6
            value=1
           }
           <at 2.7 0>
          ]}}}
       CurveIndex=$Particles.Age
       Scale Curve=$Particles.TImeofDayIntensity
      )
      Passthrough Float=&MaskFloatBySpawnGroup(
       Masked Float=&FloatFromCurve(
        FloatCurve=DataInterface() {propertyValues {Curve {keys [
            <at 0 0 RCIM_Cubic>
            {arriveTangentWeight=0.0371, interpMode=RCIM_Cubic, leaveTangentWeight=0.7205, tangentMode=RCTM_User, time=0.05, value=1}
            {
             arriveTangent=-0.0287
             arriveTangentWeight=0.1787
             interpMode=RCIM_Cubic
             leaveTangent=-0.0287
             leaveTangentWeight=0.1787
             tangentMode=RCTM_User
             tangentWeightMode=RCTWM_WeightedBoth
             time=2.3084
             value=1
            }
            <at 2.5015 0>
           ]}}}
        CurveIndex=$Particles.Age
        Scale Curve=$Particles.TImeofDayIntensity
       )
       Passthrough Float=&MaskFloatBySpawnGroup(
        Masked Float=&FloatFromCurve(
         FloatCurve=DataInterface() {propertyValues {Curve {keys [
             <at 0 0 RCIM_Cubic>
             {arriveTangentWeight=0.0363, interpMode=RCIM_Cubic, leaveTangentWeight=0.8054, tangentMode=RCTM_User, time=0.05, value=1}
             <at 2.2415 1>
             <at 2.3645 0.1473>
             {
              arriveTangent=-0.0938
              arriveTangentWeight=0.2269
              interpMode=RCIM_Cubic
              leaveTangent=-0.0938
              leaveTangentWeight=0.2269
              tangentMode=RCTM_User
              tangentWeightMode=RCTWM_WeightedBoth
              time=2.4416
              value=1
             }
             <at 2.7022 0>
            ]}}}
         CurveIndex=$Particles.Age
         Scale Curve=$Particles.TImeofDayIntensity
        )
        Passthrough Float=&MaskFloatBySpawnGroup(
         Masked Float=&FloatFromCurve(
          FloatCurve=DataInterface() {propertyValues {Curve {keys [
              {interpMode=RCIM_Cubic, leaveTangentWeight=0.0333, tangentMode=RCTM_User, time=0, value=0}
              {arriveTangentWeight=0.0333, interpMode=RCIM_Cubic, leaveTangentWeight=0.6335, tangentMode=RCTM_User, time=0.05, value=1}
              <at 2.254 1>
              <at 2.2973 0.3732>
              <at 2.3629 0.3732>
              <at 2.4285 1>
              {
               arriveTangent=-0.0244
               arriveTangentWeight=0.1394
               interpMode=RCIM_Cubic
               leaveTangent=-0.0244
               leaveTangentWeight=0.1394
               tangentMode=RCTM_User
               tangentWeightMode=RCTWM_WeightedBoth
               time=2.5883
               value=1
              }
              <at 2.7081 0>
             ]}}}
          CurveIndex=$Particles.Age
          Scale Curve=$Particles.TImeofDayIntensity
         )
         Passthrough Float=&FloatFromCurve(
          FloatCurve=DataInterface() {propertyValues {Curve {keys [
              <at 0 0 RCIM_Cubic>
              {arriveTangentWeight=0.0367, interpMode=RCIM_Cubic, leaveTangentWeight=0.7652, tangentMode=RCTM_User, time=0.05, value=1}
              <at 2.24 1>
              <at 2.3197 0.1473>
              {
               arriveTangent=-0.0636
               arriveTangentWeight=0.3335
               interpMode=RCIM_Cubic
               leaveTangent=-0.0636
               leaveTangentWeight=0.3335
               tangentMode=RCTM_User
               tangentWeightMode=RCTWM_WeightedBoth
               time=2.3581
               value=1
              }
              <at 2.623 0>
             ]}}}
          CurveIndex=$Particles.Age
          Scale Curve=$Particles.TImeofDayIntensity
         )
         Spawn Group Mask=4
        )
        Spawn Group Mask=3
       )
       Spawn Group Mask=2
      )
      Spawn Group Mask=1
     )
    )
    Scale RGB=&SelectVectorFromArray(
     Direct Array Index=&SpawnGroup()
     Vector Selection Array=DataInterface() {propertyValues {FloatData [<0.9366 0 1>, <0 1 0.0173>, <0.0313 0.7764 1>, <1 0 0.0823>, <1 0.213 0>, <1 0 0.0823>]}}
    )
    ScaleA=true
    ScaleRGB=true
   ) {enabled=false}
   &FN_Light_Attributes(
    Color=<rgba 0.5619 0.6951 1 1>
    Create Color Attribute=true
    Create Exponent Attribute=true
    Create Radius Attribute=true
    Exponent=2
    Radius=300
    Scale Color=&FloatFromCurve(
     FloatCurve=DataInterface() {propertyValues {Curve {keys [
         <at 0 0 RCIM_Cubic>
         {arriveTangentWeight=0.0581, interpMode=RCIM_Cubic, leaveTangentWeight=0.349, tangentMode=RCTM_User, time=0.0705, value=1}
         {arriveTangent=-0.0405, interpMode=RCIM_Cubic, leaveTangent=-0.0405, tangentMode=RCTM_User, time=2.2171, value=1}
         {arriveTangent=-3.885, interpMode=RCIM_Cubic, leaveTangent=-3.885, time=2.2572, value=0.225}
         {
          arriveTangent=-1.4019
          arriveTangentWeight=0.0278
          interpMode=RCIM_Cubic
          leaveTangent=0.5431
          leaveTangentWeight=0.1085
          tangentMode=RCTM_Break
          tangentWeightMode=RCTWM_WeightedBoth
          time=2.3096
          value=0.6405
         }
         {arriveTangent=-8.0089, interpMode=RCIM_Cubic, leaveTangent=-8.0089, tangentMode=RCTM_User, time=2.4648, value=0}
        ]}}}
     CurveIndex=$Particles.Age
     Scale Curve=0.3
    )
   )
   /SetVariables(Particles.LightPosition:NiagaraPosition=&AddVectorToPosition(Position=<0 0 50>, Vector=<0 0 0>)) {isSetParameters=true}
  ]
  rendererClasses [&NiagaraSpriteRendererProperties, &NiagaraLightRendererProperties]
  renderers [
   {rendererClass=&NiagaraSpriteRendererProperties}
   {rendererClass=&NiagaraLightRendererProperties, rendererIndex=1}
  ]
 }
 NE_Snowflakes {
  emitterUpdateScript [
   &EmitterState()
   &SpawnBurst_Instantaneous(Spawn Count=12, Spawn Time=1.8, Use Spawn Probability=false, UseLoopCountLimit=false) {module=SpawnBurst_Instantaneous001}
   &SpawnBurst_Instantaneous(Spawn Count=12, Use Spawn Probability=false, UseLoopCountLimit=false)
  ]
  particleSpawnScript [
   &InitializeParticle(
    Color=<rgba 0.8802 1 0.9959 1>
    Lifetime=&UniformRangedFloat(Minimum=0.7, Maximum=1.2)
    Sprite Rotation Mode=::NewEnumerator2
    Sprite Size Mode=::NewEnumerator3
    Uniform Sprite Size=&UniformRangedFloat(Minimum=14, Maximum=6)
    UsePositionOffset=false
    Write Material Random=true
   ) {module=InitializeParticle001}
   &AddVelocity(Velocity=&UniformRangedVector(Minimum=<-300 -300 -300>, Maximum=<300 300 300>))
   &ApplyInitialForces()
  ]
  particleUpdateScript [
   &ParticleState()
   &GravityForce(Gravity=<0 0 200>)
   /Niagara/Modules/Update/Forces/CurlNoiseForce(Noise Frequency=0.02, Noise Quality / Cost=::NewEnumerator0, Noise Strength=500, Pan Noise=false)
   &Drag(Drag=7.5, Use Linear Drag=true, Use Rotational Drag=true)
   &SolveForcesAndVelocity(Clamp Velocity=false, Limit Acceleration=false)
   &ScaleColor(
    Scale Alpha=&FloatFromCurve(FloatCurve=DataInterface() {propertyValues {Curve {keys [<at -0.0092 0.0144>,<at 0.0967 0.9767>,<at 1.0011 0.002>]}}})
    Scale RGB=<1 0.8 0.8>
    ScaleA=true
    ScaleRGB=true
   )
   &ScaleSpriteSize(Uniform Curve Sprite Scale=DataInterface() {propertyValues {Curve {keys [<at -0.0046 0>,<at 0.1746 0.9952>,<at 1.0011 0.5758>]}}})
   &DynamicMaterialParameters(
    Index 0 Param 4=&RandomRangeFloat(Minimum=0, Maximum=4)
    Module.Index0Param0Write=true
    Module.Index0Param1Write=true
    Module.Index0Param2Write=true
    Module.Index0Param3Write=true
   )
  ]
  rendererClasses [&NiagaraSpriteRendererProperties]
  renderers [{rendererClass=&NiagaraSpriteRendererProperties}]
 }
 NE_Sparkles {
  emitterUpdateScript [
   &EmitterState(Life Cycle Mode=::NewEnumerator1, Loop Behavior=::NewEnumerator1, Loop Delay=2, Loop Duration=3, UseLoopDelay=true)
   &SpawnRate(Spawn Probability=0.5, SpawnRate=3, Use Spawn Probability=true)
  ]
  particleSpawnScript [
   &InitializeParticle(
    Color=<rgba 5 4.0987 2.5723 1>
    Lifetime=0.3
    Sprite Rotation Mode=::NewEnumerator1
    Sprite Size Mode=::NewEnumerator1
    Sprite UV Mode=::NewEnumerator4
    Uniform Sprite Size Min=20
    UsePositionOffset=false
    Write Material Random=true
   ) {module=InitializeParticle001}
   &CameraOffset(Camera Offset Amount=10)
  ]
  particleUpdateScript [
   &SkeletalMeshLocation(
    MaskSpawnGroup=false
    Mesh Sampling Type=::NewEnumerator6
    OverrideRandomSeed=false
    Sampled Position Offset Coordinate Space:ENiagaraSkelMeshTransforms=::NewEnumerator2
    Skeletal Mesh=$System.SpheresMesh
   )
   &ShapeLocation(MaskSpawnGroup=false, Override Seed=false, Shape Origin=$Particles.Position, Sphere Radius=5, UseOwnerScale=true)
   &ParticleState()
   &SpriteRotationRate(Rotation Rate=500)
   &SolveForcesAndVelocity(Clamp Velocity=false, Limit Acceleration=false)
   &ScaleColor(
    Scale Alpha=&FloatFromCurve(FloatCurve=DataInterface() {propertyValues {Curve {keys [<at -0.0092 0.0144>,<at 0.0967 0.9767>,<at 1.0011 0.002>]}}})
    ScaleA=true
    ScaleRGB=true
   )
   &ScaleSpriteSize(Scale Factor=&Vector2DFromFloat(Value=&FloatFromCurve(FloatCurve=DataInterface() {propertyValues {Curve {keys [
          <at -0.0046 0 RCIM_Cubic>
          {
           arriveTangent=0.0048
           arriveTangentWeight=0.1184
           interpMode=RCIM_Cubic
           leaveTangent=0.0054
           leaveTangentWeight=0.5412
           tangentMode=RCTM_Break
           tangentWeightMode=RCTWM_WeightedBoth
           time=0.1214
           value=1
          }
          <at 1.0011 0.002 RCIM_Cubic>
         ]}}})))
  ]
  rendererClasses [&NiagaraSpriteRendererProperties]
  renderers [{rendererClass=&NiagaraSpriteRendererProperties}]
 }
 NE_Wisp {
  emitterUpdateScript [
   &EmitterState()
   &SpawnBurst_Instantaneous(Spawn Time=1.8, Use Spawn Probability=false, UseLoopCountLimit=false) {module=SpawnBurst_Instantaneous002}
   &SpawnBurst_Instantaneous(Spawn Group=1, Spawn Time=1.65, Use Spawn Probability=false, UseLoopCountLimit=false)
   &SpawnBurst_Instantaneous(Spawn Group=2, Spawn Time=1.7, Use Spawn Probability=false, UseLoopCountLimit=false) {module=SpawnBurst_Instantaneous001}
  ]
  particleSpawnScript [
   &InitializeParticle(
    Color=&Multiply_LinearColorByFloat(
     Linear Color=&FN_Select_Color_from_Time_Of_Day(Morning=<rgba 1 1 1 0.325>, Day=<rgba 1 1 1 0.325>, Evening=<rgba 1 1 1 0.325>, Night=<rgba 0.2691 0.4995 1 0.325>)
     Float=&FNTimeOfDayScalar(Night Scale=1, Morning Scale=1, Evening Scale=1)
    )
    Lifetime=2
    Mesh Scale=<1.2 1.2 2>
    Mesh Scale Mode=::NewEnumerator4
    UsePositionOffset=false
    Write Material Random=true
   ) {module=InitializeParticle001}
   &InitialMeshOrientation(
    Mesh Orientation Mode=::NewEnumerator1
    Rotation=&MakeVector(
     X=0
     Y=0
     Z=&MaskFloatBySpawnGroup(
      Masked Float=0.25
      Passthrough Float=&MaskFloatBySpawnGroup(Masked Float=0.5059, Passthrough Float=0.7856, Spawn Group Mask=1)
     )
    )
    Use Rotation Vector=true
   )
  ]
  particleUpdateScript [
   &ParticleState()
   &DynamicMaterialParameters(
    Index 0 Param 1=&FloatFromCurve(
     FloatCurve=DataInterface() {propertyValues {Curve {keys [
         {arriveTangent=1.2311, interpMode=RCIM_Cubic, leaveTangent=1.2311, tangentMode=RCTM_User, time=0, value=0}
         {arriveTangent=1.6164, interpMode=RCIM_Cubic, leaveTangent=1.6164, tangentMode=RCTM_User, time=1, value=1.8339}
        ]}}}
     Scale Curve=3
    )
    Index 0 Param 2=&FloatFromCurve(
     FloatCurve=DataInterface() {propertyValues {Curve {keys [
         {arriveTangent=-1.5389, interpMode=RCIM_Cubic, leaveTangent=-1.5389, tangentMode=RCTM_User, time=0, value=1}
         {arriveTangent=-0.3065, interpMode=RCIM_Cubic, leaveTangent=-0.3065, tangentMode=RCTM_User, time=0.5832, value=0}
        ]}}}
     CurveIndex=$Particles.Age
    )
    Index 0 Param 3=&FloatFromCurve(
     FloatCurve=DataInterface() {propertyValues {Curve {keys [
         {interpMode=RCIM_Cubic, tangentMode=RCTM_User, time=0.2505, value=0.1265}
         {arriveTangent=4.9567, interpMode=RCIM_Cubic, leaveTangent=4.9567, tangentMode=RCTM_User, time=0.6691, value=1}
        ]}}}
     CurveIndex=$Particles.Age
    )
    Module.Index0Param0Write=true
    Module.Index0Param1Write=true
    Module.Index0Param2Write=true
    Module.Index0Param3Write=true
   )
  ]
  rendererClasses [&NiagaraMeshRendererProperties]
  renderers [{rendererClass=&NiagaraMeshRendererProperties, rendererIndex=0}]
 }
 Smoke001 {
  emitterUpdateScript [
   &EmitterState()
   &SpawnBurst_Instantaneous(Spawn Count=11, Spawn Group=2, Spawn Time=0.55, Use Spawn Probability=false, UseLoopCountLimit=false) {module=SpawnBurst_Instantaneous002}
   &SpawnBurst_Instantaneous(Spawn Count=11, Spawn Group=1, Spawn Time=1.8, Use Spawn Probability=false, UseLoopCountLimit=false)
   &SpawnBurst_Instantaneous(Spawn Count=11, Use Spawn Probability=false, UseLoopCountLimit=false) {module=SpawnBurst_Instantaneous001}
  ]
  particleSpawnScript [
   &InitializeParticle(
    Color=&RandomRangeLinearColor(Minimum=<rgba 0.5257 0.8024 1 1>, Maximum=<rgba 0.3135 0.5073 0.6458 1>)
    Lifetime=&UniformRangedFloat(Minimum=0.4, Maximum=0.7)
    Sprite Rotation Angle=&UniformRangedFloat(Minimum=0, Maximum=1)
    Sprite Rotation Mode=::NewEnumerator1
    Sprite Size Mode=::NewEnumerator3
    Uniform Sprite Size=&UniformRangedFloat(Minimum=44, Maximum=66)
    UsePositionOffset=false
    Write Material Random=true
   ) {module=InitializeParticle001}
   /SetVariables(Particles.InitSpriteRotRate=&UniformRangedFloat(Minimum=-1, Maximum=1)) {isSetParameters=true}
   &AccelerationForce(Acceleration=&MaskVectorBySpawnGroup(
     Masked Vector=&UniformRangedVector(Minimum=<10 10 25>, Maximum=<-10 -10 0>)
     Passthrough Vector=&UniformRangedVector(Minimum=<0 0 -25>, Maximum=<0 0 -75>)
     Spawn Group Mask=2
    ))
   &ShapeLocation(
    MaskSpawnGroup=false
    Non Uniform Scale=&MaskVectorBySpawnGroup(Masked Vector=<1 1 0.1>, Passthrough Vector=<1 1 2>, Spawn Group Mask=2)
    Offset=&MaskVectorBySpawnGroup(Masked Vector=<0 0 -85>, Passthrough Vector=<0 0 0>, Spawn Group Mask=2)
    Offset Coordinate Space=::World
    Offset Mode=::NewEnumerator0
    Override Seed=false
    Sphere Radius=&MaskFloatBySpawnGroup(Masked Float=44.7227, Passthrough Float=19.667, Spawn Group Mask=2)
    UseOwnerScale=true
   )
   &ApplyInitialForces()
   &CameraOffset(Camera Offset Amount=&MaskFloatBySpawnGroup(Masked Float=5, Passthrough Float=&RandomRangeFloat(Minimum=33, Maximum=66), Spawn Group Mask=2))
   &AddVelocity(Velocity=&Multiply_Vector(
     A=$StackContext.ShapeLocation.ShapeNormal
     B=&VectorFromFloat(Value=&RandomRangeFloat(Minimum=33, Maximum=66))
    ))
  ]
  particleUpdateScript [
   &ParticleState()
   &SpriteRotationRate(Rotation Rate=&Multiply_Float(
     A=$Particles.InitSpriteRotRate
     B=&FloatFromCurve(
      FloatCurve=DataInterface() {propertyValues {Curve {keys [
          {arriveTangent=-1.9214, interpMode=RCIM_Cubic, leaveTangent=-1.9214, tangentMode=RCTM_User, time=-0.0046, value=1.0113}
          {arriveTangent=-0.2496, interpMode=RCIM_Cubic, leaveTangent=-0.2496, tangentMode=RCTM_User, time=0.9896, value=0.16}
         ]}}}
      CurveIndex=$Emitter.Age
     )
    ))
   &ScaleColor(
    Scale Alpha=&FloatFromCurve(FloatCurve=DataInterface() {propertyValues {Curve {keys [<at 0 0>,<at 0.05 1>,<at 0.7791 1>,<at 1 0>]}}}, Scale Curve=0.5)
    ScaleA=true
    ScaleRGB=true
   ) {module=ScaleColor001}
   &Drag(Drag=3, Use Linear Drag=true, Use Rotational Drag=true)
   &SolveForcesAndVelocity(Clamp Velocity=false, Limit Acceleration=false)
   &ScaleSpriteSize(Uniform Curve Sprite Scale=DataInterface() {propertyValues {Curve {keys [
        {arriveTangent=4.1785, interpMode=RCIM_Cubic, leaveTangent=4.1785, tangentMode=RCTM_User, time=0, value=0}
        {arriveTangent=1.1179, interpMode=RCIM_Cubic, leaveTangent=1.1179, tangentMode=RCTM_User, time=0.2042, value=0.6228}
        {arriveTangent=0.0224, interpMode=RCIM_Cubic, leaveTangent=0.0224, tangentMode=RCTM_User, time=1, value=1}
       ]}}})
   &DynamicMaterialParameters(
    Index 0 Param 1=&FloatFromCurve(FloatCurve=DataInterface() {propertyValues {Curve {keys [<at 0 0 RCIM_Cubic>,<at 1.0011 0.2>]}}})
    Index 0 Param 2=&FloatFromCurve(FloatCurve=DataInterface() {propertyValues {Curve {keys [
         {arriveTangent=0.0453, interpMode=RCIM_Cubic, tangentMode=RCTM_Break, time=0, value=-1}
         {arriveTangent=0.0604, interpMode=RCIM_Cubic, leaveTangent=0.0604, tangentMode=RCTM_User, time=1, value=-1}
        ]}}})
    Index 0 Param 3=&FloatFromCurve(FloatCurve=DataInterface() {propertyValues {Curve {keys [{interpMode=RCIM_Cubic, leaveTangent=4.6779, time=0.2851, value=0}]}}})
    Index 1 Param 1=&FloatFromCurve(
     FloatCurve=DataInterface() {propertyValues {Curve {keys [<at 0 0>,<at 0.25 1>]}}}
     Scale Curve=&RandomRangeFloat(Minimum=2, Maximum=1)
    )
    Index 1 Param 2=&FloatFromCurve(
     FloatCurve=DataInterface() {propertyValues {Curve {keys [
         {interpMode=RCIM_Cubic, leaveTangent=3.7422, tangentMode=RCTM_Break, time=0, value=0.2}
         {interpMode=RCIM_Cubic, leaveTangent=0.9066, time=1, value=2}
        ]}}}
     Scale Curve=&RandomRangeFloat(Minimum=0.5, Maximum=1)
    )
    Index 1 Param 3=&FloatFromCurve(FloatCurve=DataInterface() {propertyValues {Curve {keys [
         {interpMode=RCIM_Cubic, leaveTangent=2.8445, tangentMode=RCTM_Break, time=0, value=0.2}
         {interpMode=RCIM_Cubic, leaveTangent=63.2978, time=0.25, value=1}
        ]}}})
    Index 1 Param 4=&FloatFromCurve(
     FloatCurve=DataInterface() {propertyValues {Curve {keys [<at 0 1>,<at 1 0.8>]}}}
     Scale Curve=&RandomRangeFloat(Minimum=0.5, Maximum=1)
    )
    Index 2 Param 1=&FloatFromCurve(FloatCurve=DataInterface() {propertyValues {Curve {keys [{interpMode=RCIM_Cubic, leaveTangent=3.2309, tangentMode=RCTM_Break, time=0.1, value=0}, <at 1 1>]}}})
    Index 2 Param 2=$Particles.MaterialRandom
    Index 2 Param 3=&FloatFromCurve(FloatCurve=DataInterface() {propertyValues {Curve {keys [
         {interpMode=RCIM_Cubic, tangentMode=RCTM_Break, time=0, value=1}
         {interpMode=RCIM_Cubic, leaveTangent=11.1233, tangentMode=RCTM_Break, time=1, value=2}
        ]}}})
    Index 2 Param 4=0.75
    Module.Index0Param0Write=true
    Module.Index0Param1Write=true
    Module.Index0Param2Write=true
    Module.Index0Param3Write=true
    Module.Index1Param0Write=true
    Module.Index1Param1Write=true
    Module.Index1Param2Write=true
    Module.Index1Param3Write=true
    Module.Index2Param0Write=true
    Module.Index2Param1Write=true
    Module.Index2Param2Write=true
    Module.Index2Param3Write=true
    Write Parameter Index 1=true
    Write Parameter Index 2=true
   )
  ]
  rendererClasses [&NiagaraSpriteRendererProperties]
  renderers [{rendererClass=&NiagaraSpriteRendererProperties}]
 }
 Sparkles001 {
  emitterUpdateScript [
   &EmitterState()
   &SpawnBurst_Instantaneous(Spawn Count=7, Use Spawn Probability=false, UseLoopCountLimit=false)
  ]
  particleSpawnScript [
   &InitializeParticle(
    Color=<rgba 5 4.5841 1.879 1>
    Lifetime=&UniformRangedFloat(Minimum=0.3, Maximum=1)
    Sprite Size Mode=::NewEnumerator3
    Uniform Sprite Size=&UniformRangedFloat(Minimum=10, Maximum=30)
    UsePositionOffset=false
    Write Material Random=true
   ) {module=InitializeParticle001}
   &AddVelocity(Velocity=&UniformRangedVector(Minimum=<-200 -200 -200>, Maximum=<200 200 200>))
   &ApplyInitialForces()
  ]
  particleUpdateScript [
   &ParticleState()
   &GravityForce(Gravity=<0 0 200>)
   &CurlNoiseForce(Noise Frequency=0.02, Noise Quality / Cost=::NewEnumerator0, Noise Strength=500, Pan Noise=false)
   &Drag(Drag=7.5, Use Linear Drag=true, Use Rotational Drag=true)
   &SolveForcesAndVelocity(Clamp Velocity=false, Limit Acceleration=false)
   &ScaleColor(
    Scale Alpha=&FloatFromCurve(FloatCurve=DataInterface() {propertyValues {Curve {keys [<at -0.0092 0.0144>,<at 0.0967 0.9767>,<at 1.0011 0.002>]}}})
    ScaleA=true
    ScaleRGB=true
   )
   &ScaleSpriteSize(Uniform Curve Sprite Scale=DataInterface() {propertyValues {Curve {keys [<at -0.0046 0>,<at 0.1746 0.9952>,<at 1.0011 0.5432>]}}})
  ]
  rendererClasses [&NiagaraSpriteRendererProperties]
  renderers [{rendererClass=&NiagaraSpriteRendererProperties}]
 }
}
summary {
 systemName=NS_Emote_TinyTree_PropSpawn
 emitters [
  {emitterName=Smoke001, rendererClasses [&NiagaraSpriteRendererProperties]}
  {emitterName=Sparkles001, rendererClasses [&NiagaraSpriteRendererProperties]}
  {emitterName=NE_Snowflakes, rendererClasses [&NiagaraSpriteRendererProperties]}
  {
   emitterName=NE_Ornaments_Spheres
   rendererClasses [&NiagaraSpriteRendererProperties, &NiagaraLightRendererProperties]
  }
  {emitterName=NE_Wisp, rendererClasses [&NiagaraMeshRendererProperties]}
  {emitterName=NE_Sparkles, rendererClasses [&NiagaraSpriteRendererProperties]}
 ]
 userVariables [
  User.bIsFrontEnd {type=NiagaraBool}
  User.bIsFrontEndPreview {type=NiagaraBool}
  User.NormalizedNotifyProgress {type=NiagaraFloat, defaultValue=0, description=""}
  User.OrnamentOpacity {type=NiagaraFloat, defaultValue=0.8, description=""}
 ]
}
