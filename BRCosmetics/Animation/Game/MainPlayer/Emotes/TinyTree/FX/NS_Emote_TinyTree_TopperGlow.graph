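# Niagara system for the glowing tree topper in the TinyTree player emote, built from a light emitter and a sprite "bling" emitter.
# The light emitter bursts persistent light particles with a pale blue light whose scale follows an age curve, fading in after a delay.
# The bling emitter bursts a short-lived warm, spinning sprite, positioned by sampling the system's skeletal mesh with a small sphere offset.
# Curves make each sprite quickly pop in, then shrink and fade.
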NiagaraSystem /BRCosmetics/Animation/Game/MainPlayer/Emotes/TinyTree/FX/NS_Emote_TinyTree_TopperGlow
dependencies {
 usedDataInterfaces [&NiagaraDataInterfaceCurve, &NiagaraDataInterfaceSkeletalMesh]
 usedDynamicInputs []
 usedModules [
  &self:SystemSpawnScript.SystemScriptSource.SystemScriptGraph.NiagaraNodeAssignment_2.SetVariables_#1
  &self:SystemSpawnScript.SystemScriptSource.SystemScriptGraph.NiagaraNodeAssignment_3.SetVariables_#2
  &self:SystemSpawnScript.SystemScriptSource.SystemScriptGraph.NiagaraNodeAssignment_0.SetVariables_#3
  &self:SystemSpawnScript.SystemScriptSource.SystemScriptGraph.NiagaraNodeAssignment_1.SetVariables_#4
  &SystemState
  null
 ]
}
emitters {
 NE_TopperGlow_Light {
  emitterUpdateScript [
   &EmitterState(
    Inactive Response=::NewEnumerator1
    Life Cycle Mode=::NewEnumerator1
    Loop Behavior=::NewEnumerator1
    Loop Duration Mode=::NewEnumerator1
    Reset Age On Awaken=true
    Scalability Mode=::NewEnumerator1
    UseLoopDelay=false
   )
   &SpawnBurst_Instantaneous(Spawn Count=2, Use Spawn Probability=false)
  ]
  particleSpawnScript [&InitializeParticle(
    Color=<rgba 1 0.4802 0 0.25>
    Lifetime=2.2
    Sprite Size Mode=::NewEnumerator3
    Uniform Sprite Size=67.7213
    UsePositionOffset=false
    Write Material Random=true
   ) {module=InitializeParticle001}]
  particleUpdateScript [
   &ParticleState(Kill Particles When Lifetime Has Elapsed=false, Let Infinitely Lived Particles Die When Emitter Deactivates=true)
   &SolveForcesAndVelocity(Clamp Velocity=false, Limit Acceleration=false)
   &FN_Light_Attributes(
    Color=<rgba 0.5619 0.6951 1 1>
    Create Color Attribute=true
    Create Exponent Attribute=true
    Create Radius Attribute=true
    Exponent=2
    Radius=300
    Scale Color=&FloatFromCurve(
     FloatCurve=DataInterface() {propertyValues {Curve {keys [
         {arriveTangent=4.1785, interpMode=RCIM_Cubic, leaveTangent=4.1785, tangentMode=RCTM_User, time=1.6884, value=0}
         {arriveTangent=1.1179, interpMode=RCIM_Cubic, leaveTangent=1.1179, tangentMode=RCTM_User, time=2.0683, value=0.814}
         {arriveTangent=0.0224, interpMode=RCIM_Cubic, leaveTangent=0.0224, tangentMode=RCTM_User, time=2.4176, value=1}
        ]}}}
     CurveIndex=$Particles.Age
     Scale Curve=1.15
    )
   )
   /SetVariables(Particles.LightPosition:NiagaraPosition=&AddVectorToPosition(Position=<0 0 50>, Vector=<0 0 0>)) {isSetParameters=true}
  ]
  rendererClasses [&NiagaraLightRendererProperties]
  renderers [{rendererClass=&NiagaraLightRendererProperties, rendererIndex=0}]
 }
 NE_Topper_Bling {
  emitterUpdateScript [
   &EmitterState(Life Cycle Mode=::NewEnumerator1, Loop Behavior=::NewEnumerator1, Loop Delay=0.21, Loop Duration=1, UseLoopDelay=true)
   &SpawnBurst_Instantaneous(Spawn Count=1, Use Spawn Probability=false)
  ]
  particleSpawnScript [
   &InitializeParticle(
    Color=<rgba 5 4.0987 2.5723 1>
    Lifetime=0.3
    Sprite Rotation Mode=::NewEnumerator1
    Sprite Size Mode=::NewEnumerator3
    Sprite UV Mode=::NewEnumerator4
    Uniform Sprite Size=30
    UsePositionOffset=false
    Write Material Random=true
   ) {module=InitializeParticle001}
   &CameraOffset(Camera Offset Amount=10)
  ]
  particleUpdateScript [
   &SkeletalMeshLocation(
    Bone / Socket Sampling Mode=::NewEnumerator5
    MaskSpawnGroup=false
    Mesh Sampling Type=::NewEnumerator6
    Sampled Position Offset Coordinate Space:ENiagaraSkelMeshTransforms=::NewEnumerator2
    Skeletal Mesh=$System.Skeletal Mesh
   )
   &ShapeLocation(MaskSpawnGroup=false, Override Seed=false, Shape Origin=$Particles.Position, Sphere Radius=5, UseOwnerScale=true)
   &ParticleState()
   &SpriteRotationRate(Rotation Rate=500)
   &SolveForcesAndVelocity(Clamp Velocity=false, Limit Acceleration=false)
   &ScaleColor(
    Scale Alpha=&FloatFromCurve(FloatCurve=DataInterface() {propertyValues {Curve {keys [<at -0.0092 0.0144>,<at 0.0967 0.9767>,<at 1.0011 0.002>]}}})
    ScaleA=true
    ScaleRGB=true
   )
   &ScaleSpriteSize(Scale Factor=&Vector2DFromFloat(Value=&FloatFromCurve(FloatCurve=DataInterface() {propertyValues {Curve {keys [
          <at -0.0046 0 RCIM_Cubic>
          {
           arriveTangent=0.0048
           arriveTangentWeight=0.1184
           interpMode=RCIM_Cubic
           leaveTangent=0.0054
           leaveTangentWeight=0.5412
           tangentMode=RCTM_Break
           tangentWeightMode=RCTWM_WeightedBoth
           time=0.1214
           value=1
          }
          <at 1.0011 0.002 RCIM_Cubic>
         ]}}})))
  ]
  rendererClasses [&NiagaraSpriteRendererProperties]
  renderers [{rendererClass=&NiagaraSpriteRendererProperties}]
 }
}
summary {
 systemName=NS_Emote_TinyTree_TopperGlow
 emitters [
  {emitterName=NE_TopperGlow_Light, rendererClasses [&NiagaraLightRendererProperties]}
  {emitterName=NE_Topper_Bling, rendererClasses [&NiagaraSpriteRendererProperties]}
 ]
 userVariables [User.OrnamentOpacity {type=NiagaraFloat, defaultValue=0.8, description=""}]
}
